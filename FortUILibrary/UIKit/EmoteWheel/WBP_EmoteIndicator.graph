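# Emote wheel indicator widget that shows a selected segment as a material-driven icon with a number over an optional circular background.
# UpdateSelectedSegment writes the index to the icon's dynamic material and the offset segment number to the text.
# SetEnabledVisual acts only on a state change, switching enabled state, material colors, icon opacity and text visibility.
# PreConstruct calls SetLayout to show or collapse the background from Show Background.
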WidgetBlueprint /FortUILibrary/UIKit/EmoteWheel/WBP_EmoteIndicator
graphs [
 EventGraph {nodes [
   Event PreConstruct()
   > SetLayout!(execute, self, Show Background:bool=Get Show Background() -> (Show Background:bool)) -> (then)
   Event Construct()
   Event Tick()
  ]}
 SetEnabledVisual {nodes [
   n1 = SetEnabledVisual!() -> (then, bInIsEnabled:bool=~) {is=title K2Node_FunctionEntry}
   > Branch(Condition="Not Equal (Boolean)"(A=@n1, B=GetIsEnabled!(self:Widget Object Reference) -> (ReturnValue:bool)) -> (ReturnValue=~))
   > SetIsEnabled(bInIsEnabled=@n1)
   > n2 = GetDynamicMaterial(self=Get EmotePosition_Indicator() -> (EmotePosition_Indicator:Image Object Reference))
   n3 = Select(
    Option 0:Linear Color Structure=Get InactiveColor() -> (InactiveColor:Linear Color Structure)
    Option 1:Linear Color Structure=Get SelectionColor() -> (SelectionColor:Linear Color Structure)
    Index=Get bInIsEnabled() -> (bInIsEnabled:bool)
   ) -> (ReturnValue:Linear Color Structure)
   > SetVectorParameterValue(self=@n2, ParameterName=SelectedColorCenter, Value=@n3)
   > SetVectorParameterValue(self=@n2, ParameterName=SelectedColorOuter, Value=@n3)
   > SetRenderOpacity(
    self=Get EmotePosition_Indicator()
    InOpacity=Select(
     Option 0:float=Get InactiveWidgetOpacity() -> (InactiveWidgetOpacity:double)
     Option 1:float=Get ActiveWidgetOpacity() -> (ActiveWidgetOpacity:double)
     Index=Get bInIsEnabled()
    ) -> (ReturnValue:float)
   )
   > Set Visibility(
    Visibility=Select(Option 0:ESlateVisibility Enum=Collapsed, Option 1:ESlateVisibility Enum=SelfHitTestInvisible, Index=Get bInIsEnabled()) -> (ReturnValue:ESlateVisibility Enum)
    self=Get EmotePosition_Text() -> (EmotePosition_Text:Common Text Object Reference)
   )
  ]}
 SetLayout {nodes [
   n1 = SetLayout!() -> (then, Show Background:bool) {is=title K2Node_FunctionEntry}
   > SetVisibility(
    self=Get Indicator_Background() -> (Indicator_Background:Image Object Reference)
    InVisibility=Select(Option 0=Collapsed, Option 1=SelfHitTestInvisible, Index=@n1)
   )
  ]}
 UpdateSelectedSegment {nodes [
   n1 = UpdateSelectedSegment!() -> (then, Segment:int) {is=title K2Node_FunctionEntry}
   > n2 = GetDynamicMaterial(self=Get EmotePosition_Indicator())
   > SetScalarParameterValue(self=@n2, ParameterName=SelectedIndex, Value="To Float (Integer)"(InInt=@n1))
   > "SetText (Text)"(self=Get EmotePosition_Text(), InText="To Text (Integer)"(Value=int + int(A=@n1)))
  ]}
]
variables [
 ActiveWidgetOpacity {type=number, default=1}
 InactiveColor {type=LinearColor, default=<rgba 1 1 1 0.2>}
 InactiveWidgetOpacity {type=number, default=0.7}
 SelectionColor {type=LinearColor, default=<rgba 1 1 1 1>}
 Show Background {type=boolean}
]
widgets {
 info {inheritedWidgetCount=0, parentClass=&UserWidget, rootWidgetClass=&Overlay, widgetCount=4}
 widgets [
  Overlay_27 {widgetClassPath=&Overlay, widget=Overlay_27}
  Indicator_Background {
   widgetClassPath=&Image
   bIsVariable=true
   parent=Overlay_27
   widget=Indicator_Background
   properties {
    brush {
     bIsDynamicallyLoaded=false
     drawAs=Image
     imageType=FullColor
     mirroring=NoMirror
     resourceName=None
     resourceObject=/FortUILibrary/Foundation/Art_ToolKit/Tools/M_UI_Circle_White
     tiling=NoTile
     imageSize=<110 110>
     margin=<ltrb 0 0 0 0>
     outlineSettings {
      bUseBrushTransparency=false
      roundingType=HalfHeightRadius
      width=0
      color {colorUseRule=UseColor_Specified, specifiedColor=<rgba 0 0 0 0>}
      cornerRadii=<0 0 0 0>
     }
     tintColor {colorUseRule=UseColor_Specified, specifiedColor=<rgba 1 1 1 1>}
     uVRegion {bIsValid=false, max=<0 0>, min=<0 0>}
    }
    colorAndOpacity=<rgba 0 0 0 0.6>
   }
   slot {class=OverlaySlot, properties {horizontalAlignment=HAlign_Center, verticalAlignment=VAlign_Center}}
  }
  EmotePosition_Indicator {
   widgetClassPath=&Image
   bIsVariable=true
   parent=Overlay_27
   widget=EmotePosition_Indicator
   properties {brush {
     bIsDynamicallyLoaded=false
     drawAs=Image
     imageType=FullColor
     mirroring=NoMirror
     resourceName=None
     resourceObject=/FortUILibrary/UIKit/Misc/MaterialInstances/MI_UI_EmoteWheel_IndicatorIcon
     tiling=NoTile
     imageSize=<100 100>
     margin=<ltrb 0 0 0 0>
     outlineSettings {
      bUseBrushTransparency=false
      roundingType=HalfHeightRadius
      width=0
      color {colorUseRule=UseColor_Specified, specifiedColor=<rgba 0 0 0 0>}
      cornerRadii=<0 0 0 0>
     }
     tintColor {colorUseRule=UseColor_Specified, specifiedColor=<rgba 1 1 1 1>}
     uVRegion {bIsValid=false, max=<0 0>, min=<0 0>}
    }}
   slot {class=OverlaySlot, properties {horizontalAlignment=HAlign_Center, verticalAlignment=VAlign_Center}}
  }
  EmotePosition_Text {
   widgetClassPath=&CommonTextBlock
   bIsVariable=true
   parent=Overlay_27
   widget=EmotePosition_Text
   properties {
    text=1
    font {
     bForceMonospaced=false
     bMaterialIsStencil=false
     fontMaterial=null
     fontObject=&HeadingNow
     letterSpacing=0
     monospacedWidth=1
     size=24
     skewAmount=0
     typefaceFontName=75 Medium
     outlineSettings {
      bApplyOutlineToDropShadows=false
      bMiteredCorners=false
      bSeparateFillAlpha=false
      outlineMaterial=null
      outlineSize=0
      outlineColor=<rgba 0 0 0 1>
     }
    }
    shadowOffset=<0 0>
   }
   slot {class=OverlaySlot, properties {horizontalAlignment=HAlign_Center, verticalAlignment=VAlign_Center}}
  }
 ]
}
